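; Blueprint for the TroutWrist claw pickaxe that drives its cosmetic swing and impact effects.
; Swings pass the montage section index to main or offhand VFX and switch off deploy FX.
; Impacts pass that index too, while frontend previews instead spawn a directional Niagara burst in front of the pawn, using a counter and retriggerable delay.
; Weakpoint and client FX events are empty stubs; the weakspot trace and audio helpers are never called here.
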(event EventSwingRight_Common
  ;[0 0 0] swingsectionindex
  (|CallParentFunction)
  (bind _section_index (Utilities|SetupSwingMontageSectionIndex (Variables|Actor|GetInstigator))
    (:Out
      (Effects|Components|ParticleSystem|SetIntParameter (CallFunction|GetSwingVFXComponent) "SwingSectionIndex" _section_index)
      (Components|Activation|Deactivate (Variables|Default|GetDeployFX)))))

(event EventSwingLeft_Common
  ;[0 0 0] swingsectionindex
  (|CallParentFunction)
  (bind _section_index (Utilities|SetupSwingMontageSectionIndex (Variables|Actor|GetInstigator))
    (:Out
      (Effects|Components|ParticleSystem|SetIntParameter (CallFunction|GetSwingOffhandVFXComponent) "SwingSectionIndex" _section_index)
      (Components|Activation|Deactivate (Variables|Default|GetDeployFX)))))

(event FortWeapon|EventOnPlayImpactFX
  (params (HitResult :type Hit%20Result%20Structure%20%28by%20ref%29) (ImpactPhysicalSurface :type EPhysicalSurface%20Enum :default "SurfaceType_Default") (SpawnedPSC :type FXSystem%20Component%20Object%20Reference))
  (bind (_returnvalue_location _returnvalue_rotation _returnvalue_scale) (Math|Transform|BreakTransform (Transformation|GetActorTransform (Variables|Default|GetAs%20Fort%20Player%20Pawn))))
  (|CallParentFunction HitResult ImpactPhysicalSurface SpawnedPSC)
  (Variables|Default|SetImpact%20FX SpawnedPSC)
  (if (Reactive|Frontend|IsCosmeticObjectDisplayedasPreview 0)
    ;[0 0 0] frontend
    (Components|Activation|Deactivate SpawnedPSC)
    (bind _returnvalue (Niagara|SpawnSystematLocation "/BRCosmetics/Weapons/FORT_Melee/Pickaxe_TroutWrist/FX/NS_Pickaxe_TroutWrist_DirectionalImpact.NS_Pickaxe_TroutWrist_DirectionalImpact" (Math|Transform|TransformLocation (Transformation|GetActorTransform (Variables|Default|GetAs%20Fort%20Player%20Pawn)) "100.000000,0.000000,50.000000") (Math|Rotator|CombineRotators _returnvalue_rotation "0.000000,180.000000,0.000000") _returnvalue_scale true false))
    (Utilities|FlowControl|DoOnce false
      (:Completed
        (Effects|Components|ParticleSystem|SetIntParameter _returnvalue "SwingSectionIndex" (Math|Integer|%25%28Integer%29 (Variables|Default|GetFE_Counter) 4))
        (Effects|Components|ParticleSystem|SetBoolParameter _returnvalue "bIsFrontEndPreview" true)
        (Effects|Components|ParticleSystem|SetFloatParameter _returnvalue "Variant" (Math|Conversions|ToFloat(Integer) (Variables|Default|GetFE_VariantIndex)))
        (Components|Activation|Activate _returnvalue false)
        (Math|Integer|IncrementInt (Variables|Default|GetFE_Counter)
          (:%20%20
            (Utilities|FlowControl|RetriggerableDelay 0.25) ; prevents duplicated signal
          ))))
    (else
      (bind _section_index (Utilities|SetupSwingMontageSectionIndex (Variables|Actor|GetInstigator))
        (:Out
          (Effects|Components|ParticleSystem|SetIntParameter (Variables|Default|GetImpact%20FX) "SwingSectionIndex" _section_index)
          (switch Utilities|FlowControl|Switch|SwitchonEPhysicalSurface ImpactPhysicalSurface
            (:SurfaceType12
              _)
            (:SurfaceType14
              _)
            (:SurfaceType15
              _)
            (:SurfaceType16
              _))
          (CallFunction|WeakpointEvent)
          (CallFunction|CEFXServer))))))

(event Custom|WeakpointEvent)

(event Custom|CE_FX_Server
  ;[0 0 0] weakspot event- play event on client, THEN play it on every client. Allows for non-delayed impact on owner client, while keeping the visuals replicated
  (CallFunction|CEFXClient))

(event Custom|CE_FX_Client)

(event FortWeapon|EventOnUnEquippedAndDetached
  (|CallParentFunction)
  (Components|Activation|Deactivate (Variables|Default|GetDeployFX)))

(event FortWeapon|EventOnEquippedAndAttached
  ;[0 0 0] Deploy FX(Use one System for both hand, allowing to only show the FX on one of the hands!)
  (|CallParentFunction)
  (bind _asfort_player_pawn (Utilities|Casting|CastToFortPlayerPawn (Variables|Actor|GetInstigator)))
  (Variables|Default|SetAs%20Fort%20Player%20Pawn _asfort_player_pawn))

(fn SetSwingAudioParam
  (params (InInt :type Integer))
  (bind _returnvalue (AudioParameter|GetAudioParameterComponent))
  (Utilities|IsValid _returnvalue
    (:Is%20Valid
      (Audio|Parameter|SetIntegerParameter _returnvalue "SwingIdx" InInt :!class /Script/AudioExtensions.AudioParameterControllerInterface))))

(fn SetupWeakspotEvent
  (bind _outhit (Collision|LineTraceByChannel (Camera|GetCameraLocation (Game|GetPlayerCameraManager 0)) (+ (* (Math|Vector|GetForwardVector (Camera|GetCameraRotation (Game|GetPlayerCameraManager 0))) 4e+03) (Camera|GetCameraLocation (Game|GetPlayerCameraManager 0))) "TraceTypeQuery10"))
  (bind (_outhit_bblockinghit _outhit_binitialoverlap _outhit_time _outhit_distance _outhit_location _outhit_impactpoint _outhit_normal) (Collision|BreakHitResult _outhit))
  (bind _output_get (Variables|Default|SetImpactRotation _outhit_normal))
  (bind (_returnvalue_x _returnvalue_y) (Math|Vector|BreakVector (* _output_get 4e+02)))
  (Transformation|SetWorldRotation (Variables|Default|GetImpact%20FX) (Math|Conversions|RotationFromXVector (Math|Vector|MakeVector _returnvalue_x _returnvalue_y)) false false)
  (Effects|Components|ParticleSystem|SetBoolParameter (Variables|Default|GetImpact%20FX) "WeakpointHit" true)
  (Development|PrintString (Utilities|String|ToString(Vector) "0, 0, 0") true true "(R=1.000000,G=0.201102,B=0.496928,A=1.000000)"))

(fn ConstructionScript
  (|CallParentFunction))
