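# Material for Niagara mesh particles of banana bits in the TroutWrist pickaxe effects, using Default Lit shading.
# It reuses the TroutWrist emote prop's diffuse and normal textures.
# The diffuse sample is tinted by particle color, then scaled by a dynamic parameter named Emissive, letting the particle system control color and intensity per particle.
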Material /BRCosmetics/Weapons/FORT_Melee/Pickaxe_TroutWrist/FX/M_BananaBits_P
graph {
 properties {bAutomaticallySetUsageInEditor=false, bUsedWithNiagaraMeshParticles=true, usedShadingModels=Default Lit}
 nodes [
  Multiply(
   A=Multiply(
    A=ParticleColor().RGB
    B=TextureSample() {texture=/BRCosmetics/Animation/Game/MainPlayer/Emotes/TroutWrist/FX/Textures/T_TroutWrist_Emote_prop_D}.RGB
   )
   B=DynamicParameter() {paramNames [Emissive,Param2,Param3,Param4]}.Emissive
  )
  TextureSample() {samplerType=SAMPLERTYPE_Normal, texture=/BRCosmetics/Animation/Game/MainPlayer/Emotes/TroutWrist/FX/Textures/T_TroutWrist_Emote_prop_N}
 ]
}
