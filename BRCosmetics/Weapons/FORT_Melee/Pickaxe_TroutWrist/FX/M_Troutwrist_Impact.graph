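# Masked, lit material for Niagara mesh particles, apparently the liquid impact effect of the TroutWrist pickaxe.
# Dynamic parameters supply animation, erosion, fade and mask radius, while vertex color channels encode distance, thickness and isolated parts.
# The graph builds pivot-relative expansion and noise-modulated sine wobble vectors, erosion math, particle-color tint, sphere-masked Fresnel and a fade-scrolled impact texture.
# Links to material outputs are not shown.
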Material /BRCosmetics/Weapons/FORT_Melee/Pickaxe_TroutWrist/FX/M_Troutwrist_Impact
graph {
 properties {
  bAutomaticallySetUsageInEditor=false
  bUsedWithNiagaraMeshParticles=true
  blendMode=BLEND_Masked
  opacityMaskClipValue=0.01
  previewMesh=&/SM_Liquid_Impact_TroutWrist
  usedShadingModels=Default Lit
 }
 nodes [
  n1 = VertexColor()
  Add(A=@n1.R) {constB=11}
  n2 = OneMinus(None=Add(A=NamedRerouteUsage()) {constB=0.5})
  Add(A=@n2, B=NamedRerouteUsage())
  n3 = DynamicParameter() {parameterIndex=3, defaultValue=<rgba 0 1 1 1>, paramNames [Anim,Erode,Param3,Param4]}
  n4 = WorldPosition()
  n5 = Multiply(
   A=@n3.Anim
   B=Subtract(
    A=@n4.XYZ
    B=MaterialFunctionCall() {materialFunction=&ObjectPivotPoint}.Object Pivot Location
   )
  ) {constA=3}
  n6 = TransformPosition(None=@n4.XYZ) {transformSourceType=TRANSFORMPOSSOURCE_World}
  n7 = Add(A=Multiply(A=Time()) {constB=0.5}, B=NamedRerouteUsage())
  n8 = Sine(None=@n7)
  n9 = Cosine(None=@n7)
  n10 = Sine(None=@n7)
  n11 = Multiply(
   A=Multiply(
    A=@n6
    B=MaterialFunctionCall(X=@n8, Y=@n9, Z=@n10) {materialFunction=&MakeFloat3}.Result
   )
   B=NamedRerouteUsage()
  )
  Add(A=@n5, B=@n11)
  n12 = Constant() {r=3}
  n13 = MaterialFunctionCall(Exponent=@n12) {materialFunction=&MF_Fresnel_FX}
  n14 = SphereMask(
   A=ParticlePositionWS()
   B=WorldPosition().XYZ
   Radius=DynamicParameter() {defaultValue=<rgba 0.25 1 1 1>, paramNames [FadeOut,LerpRadius,Anim,Param4]}.LerpRadius
   Hardness=Constant()
  )
  Clamp(None=Multiply(A=QualitySwitch(Default=@n13.Fresnel, Low=@n12, High=@n13.Fresnel, Medium=@n13.Fresnel), B=Power(Base=@n14) {constExponent=4}))
  Constant() {r=0.2}
  Constant() {r=0.5}
  Cosine(None=@n7)
  n15 = Subtract(A=NamedRerouteUsage(), B=@n2)
  LinearInterpolate(Alpha=Clamp(None=OneMinus(None=@n15))) {constA=1, constB=-1}
  LinearInterpolate(B=@n5, Alpha=@n3.Anim) {constAlpha=1, constB=2}
  MaterialFunctionCall() {materialFunction=&MF_FX_EmissiveAdjustments}
  Multiply(A=@n11, B=NamedRerouteUsage())
  Multiply(A=@n5, B=@n11)
  Multiply(A=OneMinus(None=@n1.R))
  Multiply(A=ParticleColor().RGB, B=@n2)
  Multiply(A=@n1.R, B=@n1.R)
  n16 = VertexColor()
  NamedRerouteDeclaration(None=@n16.R) {name=Distance, nodeColor=<rgba 0 1 0.9297 1>}
  NamedRerouteDeclaration(None=@n16.G) {name=Thickness, nodeColor=<rgba 0.0312 1 0 1>}
  NamedRerouteDeclaration(None=Noise!(World Position=@n6, FilterWidth) {bTurbulence=false, levelScale=1, levels=1, outputMin=0, scale=0.01}) {name=Noise, nodeColor=<rgba 1 0 0.2109 1>}
  NamedRerouteDeclaration(None=@n16.B) {name=IsolatedParts, nodeColor=<rgba 1 0 0.3516 1>}
  NamedRerouteDeclaration(None=@n3.Erode) {name=Erode, nodeColor=<rgba 0.5937 1 0 1>}
  NamedRerouteUsage()
  ObjectPositionWS()
  n17 = Constant()
  n18 = LinearInterpolate(
   A=Constant() {r=-1}
   B=Constant() {r=1}
   Alpha=DynamicParameter() {defaultValue=<rgba 0.25 1 1 1>, paramNames [FadeOut,LerpRadius,Anim,Param4]}.FadeOut
  )
  TextureSample(UVs=Add(
    A=TextureCoordinate() {uTiling=4}
    B=MaterialFunctionCall(X=@n17, Y=@n18) {materialFunction=&MakeFloat2}.Result
   )) {texture=/Game/Effects/Fort_Effects/Textures/T_HitImpact}
 ]
}
