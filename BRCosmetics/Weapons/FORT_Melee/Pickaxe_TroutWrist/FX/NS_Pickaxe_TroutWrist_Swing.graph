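# Niagara swing trail effect for the TroutWrist pickaxe, drawing ribbons from animation trails on the weapon's skeletal mesh.
# Em_Trail uses AnimTrail modules with three FX_Tip sockets and three ribbon renderers, with spawn count selected by the swing section index.
# System TrailFade shapes alpha and can gate spawning, while time-of-day color and intensity tint the trail.
# A second, softer and wider trail emitter is disabled.
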NiagaraSystem /BRCosmetics/Weapons/FORT_Melee/Pickaxe_TroutWrist/FX/NS_Pickaxe_TroutWrist_Swing
dependencies {
 usedDataInterfaces [&NiagaraDataInterfaceCurve, &NiagaraDataInterfaceColorCurve]
 usedDynamicInputs [
  &self:SectionIndexFloatSelector_02
  &ColorFromCurve
  &FNTimeOfDayScalar
 ]
 usedModules [
  &self:SystemSpawnScript.SystemScriptSource.SystemScriptGraph.NiagaraNodeAssignment_6.SetVariables_#1
  &SystemState
  &self:SystemSpawnScript.SystemScriptSource.SystemScriptGraph.NiagaraNodeAssignment_2.SetVariables_#2
  &self:SystemSpawnScript.SystemScriptSource.SystemScriptGraph.NiagaraNodeAssignment_0.SetVariables_#3
 ]
}
emitters {
 Em_Trail {
  emitterUpdateScript [
   &EmitterState(Life Cycle Mode=::NewEnumerator1, Loop Behavior=::NewEnumerator1, Loop Duration=0.2, UseLoopDelay=false)
   &FN_SpawnAnimTrail(
    Number of Anim Trails:NiagaraInt32=&Multiply_Int(
     A=3
     B=&MakeIntFromFloat(
      Float=&self:SectionIndexFloatSelector_04(
       float:NiagaraFloat=1
       float001:NiagaraFloat=1
       float002:NiagaraFloat=1
       float003:NiagaraFloat=$System.TrailFade
       SwingIndex:NiagaraInt32=$User.SwingSectionIndex
      )
      Conversion Mode=::NewEnumerator2
     )
    )
    Spawn Interval=0
   ) {enabled=false}
   &FN_SpawnAnimTrail(
    Number of Anim Trails:NiagaraInt32=&Multiply_Int(
     A=3
     B=&MakeIntFromFloat(
      Float=&self:SectionIndexFloatSelector_04(
       float:NiagaraFloat=1
       float001:NiagaraFloat=1
       float002:NiagaraFloat=1
       float003:NiagaraFloat=$System.TrailFade
       SwingIndex:NiagaraInt32=$User.SwingSectionIndex
      )
      Conversion Mode=::NewEnumerator2
     )
    )
    Spawn Interval=0
   ) {module=FN_SpawnAnimTrail001}
  ]
  particleSpawnScript [
   /Game/Effects/Niagara/Modules/AnimTrail/FN_InitializeAnimTrail/FN_InitializeAnimTrailBoneIndecies()
   /Game/Effects/Niagara/Modules/AnimTrail/FN_InitializeAnimTrailBoneIndecies_v2/FN_InitializeAnimTrailBoneIndecies(AnimTrail Selection=1, First Bone or Socket Index=1) {module=FN_InitializeAnimTrailBoneIndecies001}
   &FN_InitializeAnimTrailBoneIndecies(AnimTrail Selection=2, First Bone or Socket Index=2, Second Bone or Socket Index=2) {module=FN_InitializeAnimTrailBoneIndecies002}
   &FN_InitializeAnimTrail(
    AnimTrailWidth=2
    Color=&MultiplyLinearColor(
     Linear Color=<rgba 1 0.8509 0.5877 1>
     Scale Factor=&MakeVector4_FromLinearColor(LinearColor=$System.ToD Color)
    )
    Lifetime=0.2
    SkeletalMesh=$System.SkeletalMesh
   )
  ]
  particleUpdateScript [
   &FN_UpdateAnimTrail(SkeletalMesh=DataInterface() {propertyValues {
      DefaultMesh=/BRCosmetics/Weapons/FORT_Melee/Pickaxe_TroutWrist/Meshes/TroutWrist_Axe
      ExcludeBoneName=root
      OptionalComponentClass=null
      PreviewMesh=&TroutWrist_Axe
      SkinningMode=SkinOnTheFly
      SoftSourceActor=null
      SourceMode=Default
      UvSetIndex=0
      WholeMeshLOD=-1
      bExcludeBone=false
      bReadDeformedGeometry=true
      bRequireCurrentFrameData=true
      ComponentTags []
      FilteredBones []
      FilteredSockets [FX_Tip_Top, FX_Tip_Bottom, FX_Tip_Mid]
      MeshUserParameter {parameter {name=None, typeDefHandle {registeredTypeIndex=1}, varData []}}
      SamplingRegions []
     }})
   &ScaleRibbonWidth()
   &ScaleColor(
    Scale Alpha=&Multiply_Float(
     A=&FloatFromCurve(
      FloatCurve=DataInterface() {propertyValues {Curve {keys [{arriveTangent=-0.5296, interpMode=RCIM_Cubic, leaveTangent=-0.5296, tangentMode=RCTM_User, time=0, value=0.25}, <at 1 0>]}}}
      Scale Curve=$System.TrailFade
     )
     B=1
    )
    Scale RGB=&VectorFromFloat(Value=$System.ToD Intensity)
    ScaleA=true
    ScaleRGB=true
   )
   &ParticleState()
  ]
  rendererClasses [&NiagaraRibbonRendererProperties]
  renderers [
   {rendererClass=&NiagaraRibbonRendererProperties}
   {rendererClass=&NiagaraRibbonRendererProperties, rendererIndex=1}
   {rendererClass=&NiagaraRibbonRendererProperties, rendererIndex=2}
  ]
 }
 Em_Trail_Soft {
  bEnabled=false
  emitterUpdateScript [
   &EmitterState(Life Cycle Mode=::NewEnumerator1, Loop Behavior=::NewEnumerator1, Loop Duration=0.2, UseLoopDelay=true)
   &FN_SpawnAnimTrail(Spawn Interval=0)
  ]
  particleSpawnScript [
   /Game/Effects/Niagara/Modules/AnimTrail/FN_InitializeAnimTrail/FN_InitializeAnimTrailBoneIndecies(First Bone or Socket Index=1, Second Bone or Socket Index=1)
   /Game/Effects/Niagara/Modules/AnimTrail/FN_InitializeAnimTrail/FN_InitializeAnimTrail(
    AnimTrailWidth=20
    Color=&MultiplyLinearColor(Linear Color=<rgba 1 1 1 1>, Scale Factor=&MakeVector4_FromLinearColor(LinearColor=$System.TODColor))
    Lifetime=2
   )
   /Game/Effects/Niagara/Modules/AnimTrail/FN_InitializeAnimTrail/FN_InitializeAnimTrail_v5/FN_InitializeAnimTrail(AnimTrailWidth=20, SkeletalMesh=$System.SkeletalMesh) {enabled=false, module=FN_InitializeAnimTrail001}
  ]
  particleUpdateScript [
   &FN_UpdateAnimTrail(SkeletalMesh=$System.SkeletalMesh)
   &ScaleRibbonWidth(Ribbon Width Scale=&FloatFromCurve(FloatCurve=DataInterface() {propertyValues {Curve {keys [
         {arriveTangent=-1.2579, interpMode=RCIM_Cubic, time=0, value=1}
         {arriveTangent=-0.9664, interpMode=RCIM_Cubic, leaveTangent=-0.9664, tangentMode=RCTM_User, time=1, value=0}
        ]}}}))
   &ScaleColor(
    Scale Alpha=&Multiply_Float(
     A=&FloatFromCurve(FloatCurve=DataInterface() {propertyValues {Curve {keys [<at 0 0.25>,<at 1 0>]}}}, Scale Curve=$System.TrailFade)
     B=4
    )
    Scale RGB=&MakeVectorFromLinearColorRGB(Color=<rgba 1 1 1 1>)
    ScaleA=true
    ScaleRGB=false
   )
   &ScaleColor(Scale RGB=&VectorFromFloat(Value=$System.ToD Intensity), ScaleA=false, ScaleRGB=true) {module=ScaleColor001}
  ]
  rendererClasses [&NiagaraRibbonRendererProperties]
  renderers [{rendererClass=&NiagaraRibbonRendererProperties}]
 }
}
summary {
 systemName=NS_Pickaxe_TroutWrist_Swing
 emitters [
  {emitterName=Em_Trail, rendererClasses [&NiagaraRibbonRendererProperties]}
  {bEnabled=false, emitterName=Em_Trail_Soft, rendererClasses [&NiagaraRibbonRendererProperties]}
 ]
 userVariables [
  User.RibbonIDOffset {type=NiagaraInt32}
  User.SkeletalMesh {
   type=NiagaraDataInterfaceSkeletalMesh
   defaultValue {
    struct=NiagaraExt_VariableValue_DataInterface
    value {dataInterface=&self:NiagaraDataInterfaceSkeletalMesh_0}
   }
  }
  User.SwingSectionIndex {type=NiagaraInt32}
 ]
}
